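# Physics asset for the Magnet vehicle body cosmetic mesh, supplying its simplified collision shape.
# A single body on the root bone uses one box, offset forward and up from the bone origin, to cover the car body.
# It defines no constraints, so the body is one rigid piece with no jointed parts.
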PhysicsAsset /VehicleCosmetics/Bodies/Magnet/Meshes/Body_Magnet_PhysicsDM
bodies {root [{shapeType=Box, extentX=444, extentY=216, extentZ=116, shapeName=root_box, center=<41.8615 0 94>}]}
constraints []
